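# Stereo MetaSound source for the diegetic Canal Beach nightclub music track.
# On play, it picks a random entry from the StartTimes array and starts the song's wave player at that offset.
# Both channels pass through a gain multiply driven by one float input, and the player's finish trigger ends the source.
# The sound also feeds the nightclub audio bus through bus and pre-effect sends.
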MetaSoundSource /Hera/Audio/Music/Diegetic/CanalBeach/MSS_CanalBeach_NightClub
inputs [
 AdditionalOperands {type=Float}
 AdditionalOperands 1 {type=Float}
 SongAsset {type=WaveAsset}
 StartTimes {type=Float:Array}
 UE.Source.OnPlay {type=Trigger}
]
interfaces [UE.OutputFormat.Stereo:1.0, UE.Source.OneShot:1.0, UE.Source:1.1]
outputs [
 "UE.OutputFormat.Stereo.Audio:0" {type=Audio}
 "UE.OutputFormat.Stereo.Audio:1" {type=Audio}
 UE.Source.OneShot.OnFinished {type=Trigger}
]
pages [Default {
  input_defaults {
   AdditionalOperands=1
   AdditionalOperands 1=1
   SongAsset=&/MX_S41_CanalBeach_NightClub_LP
   UE.Source.OnPlay=null
   StartTimes [0,33,79,122,152]
  }
  nodes [
   InitVariable.Float:Array()
   n1 = Convert.Float:Int32(Int32=Array.Num:Float:Array(Array=^StartTimes).Num)
   n2 = UE.RandomInt32(Max=Convert.Int32:Float(Float=UE.Subtract:Float(PrimaryOperand=@n1.Float) {also {defaults {AdditionalOperands=1}}}.Out).Int32) {also {version [1,1]}}
   WavePlayer_0 = UE.Wave Player:Stereo(
    Wave Asset=^SongAsset
    Play=^UE.Source.OnPlay
    Start Time=UE.ConversionFloatToTime(In=Array.Get:Float:Array(Array=^StartTimes, Index=@n2.Value).Element).Out
   ) {also {defaults {Loop="<MetaSoundBuilderResult.SUCCEEDED: 0>"}, version [1,1]}}
   Multiply_0 = UE.Multiply:Audio by Float(AdditionalOperands=^AdditionalOperands, PrimaryOperand=@WavePlayer_0.Out Right)
   Multiply_1 = UE.Multiply:Audio by Float(AdditionalOperands=^AdditionalOperands, PrimaryOperand=@WavePlayer_0.Out Left)
  ]
  links [
   {dst="@UE.OutputFormat.Stereo.Audio:1", src=Multiply_0.Out}
   {dst="@UE.OutputFormat.Stereo.Audio:0", src=Multiply_1.Out}
   {dst="@UE.Source.OneShot.OnFinished", src=WavePlayer_0.On Finished}
  ]
 }]
properties {
 assetImportData=&self:AssetImportData
 numChannels=2
 outputFormat=Stereo
 sampleRate=48000
 busSends [{audioBus=&/AB_CanalBeach_NightClub, sendLevel=1}]
 preEffectBusSends [{
   audioBus=&AB_CanalBeach_NightClub
   maxSendDistance=1000
   minSendDistance=100
   minSendLevel=0
   sendLevel=1
  }]
}
